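; FlipStream destructible prop, a slurp-tanker-style object that explodes on death to heal pawns and splash water.
; On server death, two scale-adjusted sphere overlaps build a unique actor list, which receives Curie water and, for Fort pawns, a healing GameplayEffect.
; Death effects spawn particles, a randomly rotated decal, sound and a healing indicator, and stop ambient audio.
; ShouldDie always returns true, and BeginPlay optionally overrides the mesh material.
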(event EventOnDeathServer
  (params (Damage :type Float%20%28single-precision%29) (DamageTags :type Gameplay%20Tag%20Container%20Structure%20%28by%20ref%29) (Momentum :type Vector :default "0, 0, 0") (HitInfo :type Hit%20Result%20Structure%20%28by%20ref%29) (InstigatedBy :type Controller%20Object%20Reference) (DamageCauser :type Actor%20Object%20Reference) (EffectContext :type Gameplay%20Effect%20Context%20Handle%20Structure))
  (bind _returnvalue (Transformation|GetActorScale3D))
  (bind _returnvalue_2 (Math|Vector|VectorLength _returnvalue))
  (bind _explosion_radius (Variables|Default|GetExplosion%20Radius))
  (bind _returnvalue_3 (* _explosion_radius _returnvalue_2))
  (bind _allhitactors (Variables|Default|GetAllHitActors))
  (if (Class|AthenaFunctionLibrary|ShouldAllowDeathEffects DamageTags)
    (CallFunction|DeactivateEmitter)
    (Variables|Default|SetWaterInstigator InstigatedBy)
    (bind (_outactors _returnvalue_4) (Collision|SphereOverlapActors (Transformation|GetWorldLocation (Variables|Default|GetExplosionLocation01)) _returnvalue_3 (Variables|Default|GetOverlap%20Object%20Types)))
    (for _array_element _outactors
      (Utilities|Array|AddUnique _allhitactors _array_element))
    (bind (_outactors_2 _returnvalue_5) (Collision|SphereOverlapActors (Transformation|GetWorldLocation (Variables|Default|GetExplosionLocation02)) _returnvalue_3 (Variables|Default|GetOverlap%20Object%20Types) 0 _allhitactors))
    (for _array_element_2 _outactors_2
      (Utilities|Array|AddUnique (Variables|Default|GetAllHitActors) _array_element_2))
    (if (or _returnvalue_4 _returnvalue_5)
      (for _array_element_3 (Variables|Default|GetAllHitActors)
        (Utilities|IsValid _array_element_3
          (:Is%20Valid
            (bind _output_get (Variables|Default|SetHitActor _array_element_3))
            (CallFunction|ApplyWaterifCurieEnabled (Variables|Default|GetWaterInstigator) _output_get)
            (bind _asfort_pawn (Utilities|Casting|CastToFortPawn (Variables|Default|GetHitActor)))
            (Utilities|IsValid _asfort_pawn
              (:Is%20Valid
                (GameplayEffects|FortApplyGameplayEffectToTarget :self (Ability|GetFortAbilitySystemComponent self) :GameplayEffectClass (Variables|Default|GetGE_HealingEffect) :TargetOfEffect (Ability|GetAbilitySystemComponent _asfort_pawn) :OptionalSourceObject self :OptionalEffectCauser self :Level 0.0)))))))))

(event EventBeginPlay
  (|CallParentFunction)
  (if (Variables|Default|GetAutomatically%20Set%20Material)
    (Rendering|Material|SetMaterial (Variables|Debug|GetStaticMeshComponent) 0 (Variables|Default|GetMaterialInstanceOverride))))

(event EventOnDeathPlayEffects
  (params (Damage :type Float%20%28single-precision%29) (DamageTags :type Gameplay%20Tag%20Container%20Structure%20%28by%20ref%29) (Momentum :type Vector :default "0, 0, 0") (HitInfo :type Hit%20Result%20Structure%20%28by%20ref%29) (InstigatedBy :type Fort%20Pawn%20Object%20Reference) (DamageCauser :type Actor%20Object%20Reference) (EffectContext :type Gameplay%20Effect%20Context%20Handle%20Structure))
  (bind _explosionlocation01 (Variables|Default|GetExplosionLocation01))
  (bind _returnvalue (Transformation|GetWorldLocation _explosionlocation01))
  (bind _returnvalue_2 (+ _returnvalue "382.000000, 0, 0"))
  (|CallParentFunction Damage DamageTags Momentum HitInfo InstigatedBy DamageCauser EffectContext)
  (Effects|SpawnEmitteratLocation "/Game/Athena/Items/Gameplay/SilkyTrombone/FX/P_SilkyTrombone_Explode_Tank.P_SilkyTrombone_Explode_Tank" (Transformation|GetSocketLocation (Variables|Debug|GetStaticMeshComponent) "DeathFX") (Transformation|GetSocketRotation (Variables|Debug|GetStaticMeshComponent) "DeathFX") "1.000000,1.000000,1.000000" true "AutoRelease")
  (bind _returnvalue_3 (Rendering|Decal|SpawnDecalatLocation "/WildEstate/Gameplay/NeonBurn/Environmental/Materials/MD_Decal_NeonBurn_ChillBronco_Hit.MD_Decal_NeonBurn_ChillBronco_Hit" (Math|Vector|MakeVector (* (.x (Variables|Default|GetDeath%20Decal%20Size)) (.x (Transformation|GetActorScale3D))) (* (.y (Variables|Default|GetDeath%20Decal%20Size)) (.y (Transformation|GetActorScale3D))) (* (.z (Variables|Default|GetDeath%20Decal%20Size)) (.z (Transformation|GetActorScale3D)))) (- (Transformation|GetWorldLocation (Variables|Default|GetParticleTransformSource)) "0, 0,250.000000") (Math|Rotator|MakeRotator (Math|Random|RandomFloatinRangefromStream (Variables|Default|GetRandom%20Stream) 0.0 359.0) -9e+01)))
  (Rendering|Components|Decal|SetFadeIn _returnvalue_3 0.0 0.0)
  (Rendering|Components|Decal|SetFadeOut _returnvalue_3 0.25 1.75 true)
  (Audio|PlaySoundatLocation "/Game/Athena/Sounds/Items/Slurp/Slurp_Tanker_Explode_Cue.Slurp_Tanker_Explode_Cue" _returnvalue_2)
  (Audio|Components|Audio|Stop (Variables|Default|GetAmbientAudio))
  (Fortnite|ShowSoundIndicatorLocal :Location _returnvalue_2 :MaxAudibleDistance 2.5e+03 :IndicatorType "Healing" :Instigator self :Tint "(R=0.000000,G=0.000000,B=0.000000,A=0.000000)" :AttenuationMaxRangeMultiplier 1.0 :AttenuationIntensityMultiplier 1.0))

(fn ApplyWaterIfCurieEnabled
  (params (Instigator :type Actor%20Object%20Reference) (Target :type Actor%20Object%20Reference))
  (Networking|CanExecuteCosmeticEvents
    (:False
      (if (Curie|IsCurieActive)
        (Utilities|IsValid Target
          (:Is%20Valid
            (Curie|ExecuteWaterApplication Instigator self Target 1e+02 (Collision|MakeHitResult false true 0.0 0.0 (Transformation|GetActorLocation) (Transformation|GetActorLocation) "0,0,1" "0,0,1" 0 Target) (Variables|Default|GetLandscapeExtinguishRadius))))))))

(fn GetMarkerDetailsId
  (return (|CallParentFunction)))

(fn ShouldDie
  (params (Damage :type Float%20%28single-precision%29) (EventInstigator :type Controller%20Object%20Reference) (DamageCauser :type Actor%20Object%20Reference) (bIsTest :type Boolean) (EffectContext :type Gameplay%20Effect%20Context%20Handle%20Structure%20%28by%20ref%29))
  (return true))

(fn ConstructionScript
  (|CallParentFunction))
